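# Niagara skydiving contrail cosmetic "Realm": three small ship meshes follow the player's skeletal mesh with wobble, banking and velocity-based catch-up offsets.
# Ribbon air-edge and trail emitters spawn from ship particles via attribute readers, lerping toward live ship positions to stay attached.
# System Dive drives spawn rates, trail lifetime, orientation response and GPU speed-line sprites.
# Separate FE emitters handle front-end display, switched by front-end modifiers, with added backward velocity faking motion.
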NiagaraSystem /BRCosmetics/Athena/Effects/Particles/Skydiving/NS_Contrail_Realm/NS_Realm_Contrail
dependencies {
 usedDataInterfaces [
  &NiagaraDataInterfaceCurve
  &NiagaraDataInterfaceArrayFloat3
  &NiagaraDataInterfaceVector2DCurve
  &NiagaraDataInterfaceParticleRead
  &NiagaraDataInterfaceSkeletalMesh
 ]
 usedDynamicInputs [
  &Vector2DFromCurve
  &FN_FrontEndVector3Modifier
  &Divide_Float
 ]
 usedModules [
  &self:SystemSpawnScript.SystemScriptSource.SystemScriptGraph.NiagaraNodeAssignment_0.SetVariables_#1
  &self:SystemSpawnScript.SystemScriptSource.SystemScriptGraph.NiagaraNodeAssignment_2.SetVariables_#2
  &SystemState
  /Game/Effects/Niagara/Modules/Skydiving/FN_DivingAttributes/v6/FN_DivingAttributes
  &self:ScratchModule
  &self:SystemSpawnScript.SystemScriptSource.SystemScriptGraph.NiagaraNodeAssignment_4.SetVariables_#3
  &self:SystemSpawnScript.SystemScriptSource.SystemScriptGraph.NiagaraNodeAssignment_1.SetVariables_#4
  &FN_Integrator
  &Timeline
  &self:SystemSpawnScript.SystemScriptSource.SystemScriptGraph.NiagaraNodeAssignment_7.SetVariables_#5
  &self:SystemSpawnScript.SystemScriptSource.SystemScriptGraph.NiagaraNodeAssignment_9.SetVariables_#6
 ]
}
emitters {
 Air_Edges {
  emitterSpawnScript [/SetVariables(Emitter.X Offset=-8) {isSetParameters=true, module=SetVariables_#7}]
  emitterUpdateScript [
   &EmitterState()
   &SpawnParticlesFromOtherEmitter(
    Attribute Reader=DataInterface() {propertyValues {EmitterBinding {bindingMode=Other, emitterName=ship}}}
    Spawn Rate=&FN_FrontEndFloatModifier(Gameplay=&Lerp_Float(A=150, B=900, Alpha=$System.Dive), Front End=0)
    Spawn Rate Per Particle Cap=100000000
    Use Maximum Age=false
    Use Minimum Age=false
   )
  ]
  particleSpawnScript [
   &InitializeParticle(Color=<rgba 1 1 1 0.0175>, Lifetime=0.068, UsePositionOffset=false, Write Material Random=true)
   &SampleParticlesFromOtherEmitter(
    Applied Position Offset:Vector3f=<0 0 -5>
    Mesh Orientation Sampling:ENiagara_DirectReadSamplingMode=::NewEnumerator0
    Particle ID Sampling=::NewEnumerator1
   )
   /SetVariables(
    Particles.RibbonFacing=&ReturnMeshOrientationAxis(Mesh Orientation Axis=::ZAxis, Mesh Orientation Quat=$Particles.MeshOrientation)
    Particles.RibbonWidth=$System.Ship Scale
   ) {isSetParameters=true, module=SetVariables_#8}
   /SetVariables(Particles.Position=&Add_Vector(
     A=Linked() {linkedVariable {name=Particles.Position, type=NiagaraPosition}}
     B=&Multiply_VectorByFloat(
      Vector=&ReturnMeshOrientationAxis(Mesh Orientation Axis=::XAxis, Mesh Orientation Quat=$Particles.MeshOrientation)
      Float=$Emitter.X Offset
     )
    )) {isSetParameters=true, module=SetVariables_#9}
  ]
  particleUpdateScript [
   &ParticleState()
   &SolveForcesAndVelocity(Clamp Velocity=false, Limit Acceleration=false)
   &self:SampleShip(
    New Particle Attribute Reader=DataInterface() {propertyValues {EmitterBinding {bindingMode=Other, emitterName=Ship}}}
    New int32:NiagaraInt32=&MakeIntFromFloat(Float=$Particles.SampleParticlesFromOtherEmitter.SampledIndex)
   ) {module=SampleShip}
   /SetVariables(Particles.LerpPos:NiagaraPosition=/Niagara/DynamicInputs/Lerp/Lerp_Position(
     A:NiagaraPosition=&Add_Vector(
      A=Linked() {linkedVariable {name=Output.SampleShip.Ship Pos, type=NiagaraPosition}}
      B=&Multiply_VectorByFloat(
       Vector=&ReturnMeshOrientationAxis(Mesh Orientation Axis=::XAxis, Mesh Orientation Quat=$Particles.MeshOrientation)
       Float=$Emitter.X Offset
      )
     )
     B:NiagaraPosition=$Particles.Position
     Alpha:NiagaraFloat=&FloatFromCurve(FloatCurve=DataInterface() {propertyValues {Curve {keys [{tangentMode=RCTM_User, time=0.1, value=0}, {tangentMode=RCTM_User, time=1, value=1}]}}})
    )) {isSetParameters=true, module=SetVariables_#10}
   /SetVariables(Particles.Rotate Axis:Vector3f=&Lerp_Vector(
     A=&ReturnMeshOrientationAxis(Mesh Orientation Axis=::ZAxis, Mesh Orientation Quat=$Output.SampleShip.Ship Rot)
     B=&ReturnMeshOrientationAxis(Mesh Orientation Axis=::ZAxis, Mesh Orientation Quat=$Particles.MeshOrientation)
     Alpha=&FloatFromCurve(FloatCurve=DataInterface() {propertyValues {Curve {keys [{tangentMode=RCTM_User, time=0, value=0}, {tangentMode=RCTM_User, time=1, value=1}]}}})
    )) {isSetParameters=true}
   /Niagara/Modules/Math/FindClosestPointOnLineSegment(
    Line Segment End:Vector3f=&Add_Vector(
     A=Linked() {linkedVariable {name=Output.SampleShip.Ship Pos, type=NiagaraPosition}}
     B=&Multiply_VectorByFloat(
      Vector=&ReturnMeshOrientationAxis(Mesh Orientation Axis=::XAxis, Mesh Orientation Quat=$Particles.MeshOrientation)
      Float=25
     )
    )
    Line Segment Start:NiagaraPosition=$Output.SampleShip.Ship Pos
   )
   /SetVariables(Particles.Line Pos:NiagaraPosition=&Lerp_Position(
     A:NiagaraPosition=Linked() {linkedVariable {name=Output.FindClosestPointOnLineSegment.Closest Point on Line Segment, type=Vector3f}}
     B:NiagaraPosition=$Particles.LerpPos
     Alpha:NiagaraFloat=&FloatFromCurve(FloatCurve=DataInterface() {propertyValues {Curve {keys [{arriveTangent=5.0021, time=0, value=0}, <at 1 1 RCIM_Cubic>]}}})
    )) {isSetParameters=true}
  ]
  rendererClasses [&NiagaraRibbonRendererProperties]
  renderers [{rendererClass=&NiagaraRibbonRendererProperties}]
 }
 Air_Edges_FE {
  emitterSpawnScript [/SetVariables(Emitter.X Offset=-8) {isSetParameters=true, module=SetVariables_#7}]
  emitterUpdateScript [
   &EmitterState()
   &SpawnParticlesFromOtherEmitter(
    Attribute Reader=DataInterface() {propertyValues {EmitterBinding {bindingMode=Other, emitterName=ship_FE}}}
    Spawn Rate=&FN_FrontEndFloatModifier(Gameplay=0, Front End=&Lerp_Float(A=150, B=900, Alpha=$System.Dive))
    Spawn Rate Per Particle Cap=100000000
    Use Maximum Age=false
    Use Minimum Age=false
   )
  ]
  particleSpawnScript [
   &InitializeParticle(Color=<rgba 1 1 1 0.0175>, Lifetime=0.068, UsePositionOffset=false, Write Material Random=true)
   &SampleParticlesFromOtherEmitter(
    Applied Position Offset:Vector3f=<0 0 -5>
    Mesh Orientation Sampling:ENiagara_DirectReadSamplingMode=::NewEnumerator0
    Particle ID Sampling=::NewEnumerator1
   )
   /SetVariables(
    Particles.RibbonFacing=&ReturnMeshOrientationAxis(Mesh Orientation Axis=::ZAxis, Mesh Orientation Quat=$Particles.MeshOrientation)
    Particles.RibbonWidth=$System.Ship Scale
   ) {isSetParameters=true, module=SetVariables_#8}
   /SetVariables(Particles.Position=&Add_Vector(
     A=Linked() {linkedVariable {name=Particles.Position, type=NiagaraPosition}}
     B=&Multiply_VectorByFloat(
      Vector=&ReturnMeshOrientationAxis(Mesh Orientation Axis=::XAxis, Mesh Orientation Quat=$Particles.MeshOrientation)
      Float=$Emitter.X Offset
     )
    )) {isSetParameters=true, module=SetVariables_#9}
   &AddVelocity(
    Rotation Coordinate Space=::World
    Velocity=&ReturnMeshOrientationAxis(Mesh Orientation Axis=::XAxis, Mesh Orientation Quat=$Particles.MeshOrientation)
    Velocity Speed Scale=-1500
   )
  ]
  particleUpdateScript [
   &ParticleState()
   &SolveForcesAndVelocity(Clamp Velocity=false, Limit Acceleration=false)
   &self:SampleShip(
    New Particle Attribute Reader=DataInterface() {propertyValues {EmitterBinding {bindingMode=Other, emitterName=ship_FE}}}
    New int32:NiagaraInt32=&MakeIntFromFloat(Float=$Particles.SampleParticlesFromOtherEmitter.SampledIndex)
   ) {module=SampleShip}
   /SetVariables(Particles.LerpPos:NiagaraPosition=&Lerp_Position(
     A:NiagaraPosition=&Add_Vector(
      A=Linked() {linkedVariable {name=Output.SampleShip.Ship Pos, type=NiagaraPosition}}
      B=&Multiply_VectorByFloat(
       Vector=&ReturnMeshOrientationAxis(Mesh Orientation Axis=::XAxis, Mesh Orientation Quat=$Particles.MeshOrientation)
       Float=$Emitter.X Offset
      )
     )
     B:NiagaraPosition=$Particles.Position
     Alpha:NiagaraFloat=&FloatFromCurve(FloatCurve=DataInterface() {propertyValues {Curve {keys [{tangentMode=RCTM_User, time=0.1, value=0}, {tangentMode=RCTM_User, time=1, value=1}]}}})
    )) {isSetParameters=true, module=SetVariables_#10}
  ]
  rendererClasses [&NiagaraRibbonRendererProperties]
  renderers [{rendererClass=&NiagaraRibbonRendererProperties}]
 }
 FN_SpeedLines {
  simTarget=GPUComputeSim
  emitterUpdateScript [
   &EmitterState(Life Cycle Mode=::NewEnumerator1, Loop Duration=1, UseLoopDelay=false)
   &SpawnRate(
    Spawn Probability=&MakeCustomFloatFromBool(Bool=&SetBoolByFloatComparison(A=$System.Age, B=0.2), True Float=1, False Float=1)
    SpawnRate=&Lerp_Float(A=0, B=&FN_FrontEndFloatModifier(Gameplay=55, Front End=0), Alpha=$System.Dive)
    Use Spawn Probability=true
   )
   &SpawnRate(
    Spawn Probability=&MakeCustomFloatFromBool(Bool=&SetBoolByFloatComparison(A=$System.Age, B=0.2), True Float=1, False Float=1)
    SpawnRate=&Lerp_Float(A=0, B=500, Alpha=$System.Dive)
    Use Spawn Probability=true
   ) {enabled=false, module=SpawnRate001}
  ]
  particleSpawnScript [
   &InitializeParticle(
    Color Maximum=<rgba 0.25 0.25 0.25 0.05>
    Color Minimum=<rgba 1 1 1 0.025>
    Color Mode=::NewEnumerator2
    Lifetime Max=0.5
    Lifetime Min=0.3
    Lifetime Mode=::NewEnumerator1
    Sprite Rotation Mode=::NewEnumerator2
    Sprite Size=&UniformRangedVector2D(Minimum=<7 100>, Maximum=<1 200>)
    Sprite Size Mode=::NewEnumerator4
    UsePositionOffset=false
    Write Material Random=true
   ) {module=InitializeParticle001}
   &SkeletalMeshLocation(
    Bone / Socket Index=3
    Bone Sampling Mode=::NewEnumerator4
    MaskSpawnGroup=false
    Sampled Orientation Coordinate Space=::Local
    Skeletal Mesh=$System.SkeletalMesh
   )
   &SphereLocation(
    MaskSpawnGroup=false
    Non Uniform Scale=&RotateVectorByQuaternion(Vector To Rotate=<1 1 1>, Quaternion=$Particles.SkeletalMeshLocation.SampledRotation)
    Offset=&Add_Vector(
     A=&RotateVectorByQuaternion(Vector To Rotate=<100 0 0>, Quaternion=$Particles.SkeletalMeshLocation.SampledRotation)
     B=&TransformVector(Vector=<0 0 -200>, Source Space=::World, Destination Space=::World)
    )
    Override Seed=false
    Sphere Orientation Axis:Vector3f=&RotateVectorByQuaternion(Vector To Rotate=<100 0 0>, Quaternion=$Particles.SkeletalMeshLocation.SampledRotation)
    Sphere Origin:NiagaraPosition=$Particles.SkeletalMeshLocation.SampledPosition
    Sphere Radius=110
    Surface Only=false
    UseNonUniformScale=true
    UseOffset=true
    UseSphereOrientation=true
   )
   &InheritVelocity(Module.Limit Speed=true)
  ]
  particleUpdateScript [
   &ParticleState()
   &SolveForcesAndVelocity(Clamp Velocity=false, Limit Acceleration=false)
   &ScaleSpriteSize(
    Non-Uniform Curve Scale=<5 3>
    Non-Uniform Curve Sprite Scale=DataInterface() {propertyValues {
      XCurve {keys [<at -0.0069 0.5 RCIM_Cubic>,<at 0.4918 1 RCIM_Cubic>,<at 1 0.5 RCIM_Cubic>]}
      YCurve {keys [
        {
         arriveTangent=3.1787
         arriveTangentWeight=0.684
         interpMode=RCIM_Cubic
         leaveTangent=3.1787
         leaveTangentWeight=0.684
         tangentMode=RCTM_User
         tangentWeightMode=RCTWM_WeightedBoth
         time=0
         value=0.25
        }
        {
         arriveTangentWeight=0.2354
         interpMode=RCIM_Cubic
         leaveTangentWeight=0.3527
         tangentMode=RCTM_Break
         tangentWeightMode=RCTWM_WeightedBoth
         time=0.4975
         value=1
        }
        <at 1 0.25 RCIM_Cubic>
       ]}
     }}
    Scale Sprite Size Mode=::NewEnumerator3
   )
  ]
  rendererClasses [&NiagaraSpriteRendererProperties]
  renderers [{rendererClass=&NiagaraSpriteRendererProperties}]
 }
 Ship {
  emitterUpdateScript [
   &EmitterState()
   &SpawnBurst_Instantaneous(
    Age:NiagaraFloat=$Emitter.Age
    Spawn Count=&MakeIntFromFloat(Float=&FN_FrontEndFloatModifier(Gameplay=1, Front End=0))
    Use Spawn Probability=false
    UseLoopCountLimit=false
   )
   &SpawnBurst_Instantaneous(
    Age:NiagaraFloat=$Emitter.Age
    Spawn Count=&MakeIntFromFloat(Float=&FN_FrontEndFloatModifier(Gameplay=1, Front End=0))
    Spawn Group=1
    Use Spawn Probability=false
    UseLoopCountLimit=false
   ) {module=SpawnBurst_Instantaneous002}
   &SpawnBurst_Instantaneous(
    Age:NiagaraFloat=$Emitter.Age
    Spawn Count=&MakeIntFromFloat(Float=&FN_FrontEndFloatModifier(Gameplay=1, Front End=0))
    Spawn Group=2
    Use Spawn Probability=false
    UseLoopCountLimit=false
   ) {module=SpawnBurst_Instantaneous001}
   &InterpolateOverTime(Rate Of Change=50, Target Vector Value:Vector3f=$Engine.Owner.Velocity, Variable Type=::NewEnumerator2)
  ]
  particleSpawnScript [
   &InitializeParticle(
    Color Mode=::NewEnumerator0
    Lifetime=5
    Mesh Scale Mode=::NewEnumerator3
    Mesh Uniform Scale=$System.Ship Scale
    Position Offset=<0 0 1000>
    UsePositionOffset=true
    Write Material Random=true
   )
   &ShapeLocation(MaskSpawnGroup=false, Override Seed=false, UseOwnerScale=true)
   &InitialMeshOrientation(Mesh Orientation Mode=::NewEnumerator7, Use Rotation Vector=true)
   /SetVariables(Particles.Debug sprite offset=50) {isSetParameters=true, module=SetVariables_#11}
  ]
  particleUpdateScript [
   &ParticleState(Kill Particles When Lifetime Has Elapsed=false, Let Infinitely Lived Particles Die When Emitter Deactivates=true)
   /SetVariables(Particles.Ambient:Vector3f=&Multiply_VectorByFloat(
     Vector=&MakeVector(
      X=&Waveform(
       "Waveform [1]"=::NewEnumerator12
       "[1] Input Position"=&RandomRangeFloat(Minimum=0, Maximum=1000)
       "[1] Frequency"=1
       "[1] Amplitude Min/Max"=<-50 50>
      )
      Y=&Waveform(
       "Waveform [1]"=::NewEnumerator12
       "[1] Input Position"=&RandomRangeFloat(Minimum=1200, Maximum=1800)
       "[1] Frequency"=1
       "[1] Amplitude Min/Max"=<-50 50>
      )
      Z=&Waveform(
       "Waveform [1]"=::NewEnumerator12
       "[1] Input Position"=&RandomRangeFloat(Minimum=2200, Maximum=2800)
       "[1] Frequency"=1
       "[1] Amplitude Min/Max"=<-50 50>
      )
     )
     Float=0.35
    )) {isSetParameters=true, module=SetVariables_#12}
   &InterpolateOverTime(
    Rate Of Change=3
    Target Float Value:NiagaraFloat=&MakeCustomFloatFromBool(Bool=$System.Timeline001.IsRewinding, True Float=0, False Float=1)
   ) {module=InterpolateOverTime004}
   /SetVariables(Particles.CatchupVector:Vector3f=&Multiply_VectorByFloat(
     Vector=&Multiply_VectorByFloat(
      Vector=&NormalizeVector(Vector To Normalize=$System.SkydiveVelocity, Safe Normalize=true, Safe Fallback Threshold=0, Safe Fallback Vector=<0 0 0.001>)
      Float=-75
     )
     Float=&Lerp_Float(
      A=&FloatFromCurve(
       FloatCurve=DataInterface() {propertyValues {Curve {keys [
           {interpMode=RCIM_Cubic, tangentMode=RCTM_User, time=0, value=0}
           {interpMode=RCIM_Cubic, tangentMode=RCTM_User, time=0.35, value=-0.35}
           {tangentMode=RCTM_User, time=1, value=0}
          ]}}}
       CurveIndex=$System.Timeline001.NormalizedTime
      )
      B=&FloatFromCurve(
       FloatCurve=DataInterface() {propertyValues {Curve {keys [
           {arriveTangent=3.3827, arriveTangentWeight=1.1106, interpMode=RCIM_Cubic, leaveTangentWeight=1.1106, time=0, value=0}
           {arriveTangentWeight=0.5241, interpMode=RCIM_Cubic, leaveTangentWeight=0.1667, time=0.5, value=0.65}
           <at 1 0 RCIM_Cubic>
          ]}}}
       CurveIndex=$System.Timeline001.NormalizedTime
      )
      Alpha=$Particles.InterpolateOverTime004.InterpolateOverTime_Float.Moving Average
     )
    )) {isSetParameters=true}
   &SkeletalMeshLocation(
    Bone Sampling Mode=::NewEnumerator4
    MaskSpawnGroup=false
    Orientation Sampling=::NewEnumerator1
    Sampled Position Offset=&Add_Vector(
     A=$Particles.CatchupVector
     B=/Niagara/DynamicInputs/Vectors/RotateVector(
      Vector To Rotate:Vector3f=&Add_Vector(
       A=$Particles.Ambient
       B=&Add_Vector(
        A=&SelectVectorFromArray(
         Direct Array Index=&ReturnExecIndex()
         Direct Array Mode=::NewEnumerator1
         Vector Selection Array=DataInterface() {propertyValues {FloatData [<0 75 150>, <-50 0 175>, <50 0 176>]}}
        )
        B=$System.OverallTargetOffset
       )
      )
      Yaw:NiagaraFloat=&Waveform(
       "Waveform [1]"=::NewEnumerator12
       "[1] Input Position"=&RandomRangeFloat(Minimum=200000, Maximum=100000)
       "[1] Frequency"=1.25
       "[1] Amplitude Min/Max"=<7 -7>
      )
      Pitch:NiagaraFloat=0
      Roll:NiagaraFloat=0
      Rotation Rate:NiagaraFloat=1
     )
    )
    Skeletal Mesh=$System.SkeletalMesh
    Use Sampled Mesh Attribute as Ribbon ID=false
   )
   &ScaleMeshSize(Scale Factor=&VectorFromFloat(Value=&FloatFromCurve(
      FloatCurve=DataInterface() {propertyValues {Curve {keys [{arriveTangent=1.0186, interpMode=RCIM_Cubic, leaveTangent=1.0186, tangentMode=RCTM_User, time=0, value=0}, <at 1 1 RCIM_Cubic>]}}}
      CurveIndex=$Particles.Age
     )))
   &DynamicMaterialParameters(
    Index 0 Param 1=1
    Module.Param0WriteEnabled=true
    Module.Param1WriteEnabled=true
    Module.Param2WriteEnabled=true
    Module.Param3WriteEnabled=true
   )
   &Drag(Drag=2, Use Linear Drag=true, Use Rotational Drag=false)
   &SolveForcesAndVelocity(Clamp Velocity=false, Limit Acceleration=false)
   &UpdateMeshOrientation(
    Bank Rate:NiagaraFloat=&Lerp_Float(
     A=&Lerp_Float(
      A=1
      B=30
      Alpha=&Waveform(
       "Waveform [1]"=::NewEnumerator12
       "[1] Input Position"=&RandomRangeFloat(Minimum=0, Maximum=1000)
       "[1] Phase Scale"=50000
       "[1] Frequency"=0.5
       "[1] Amplitude Min/Max"=<0 1>
      )
     )
     B=900
     Alpha=$System.Dive
    )
    Banking Angle Min / Max:Vector2f=&Lerp_Vector2(A=<-10 10>, B=<-1 1>, Alpha=$System.Dive)
    Orientation Method=::NewEnumerator3
    Particle Velocity:Vector3f=$Emitter.InterpolateOverTime.InterpolateOverTime_V3_.Moving Average
    Rotation Decay Rate In Degrees:NiagaraFloat=&Lerp_Float(A=150, B=500, Alpha=$System.Dive)
    Up Vector:Vector3f=&Lerp_Vector(
     A=&ReturnMeshOrientationAxis(Mesh Orientation Axis=::YAxis, Mesh Orientation Quat=$Particles.SkeletalMeshLocation.SampledRotation)
     B=&Multiply_VectorByFloat(
      Vector=&ReturnMeshOrientationAxis(Mesh Orientation Axis=::ZAxis, Mesh Orientation Quat=$Particles.SkeletalMeshLocation.SampledRotation)
      Float=1
     )
     Alpha=&FloatFromCurve(FloatCurve=DataInterface() {propertyValues {Curve {keys [<at 0 0.5>,<at 0.25 0.75>]}}}, CurveIndex=$System.Dive)
    )
    Upright Roll Correction Speed:NiagaraFloat=&Lerp_Float(A=25, B=90000, Alpha=$System.Dive)
    Velocity Orientation Convergence Rate:NiagaraFloat=&Lerp_Float(
     A=&Lerp_Float(
      A=3
      B=12
      Alpha=&Waveform(
       "Waveform [1]"=::NewEnumerator12
       "[1] Input Position"=&RandomRangeFloat(Minimum=0, Maximum=100)
       "[1] Frequency"=5
       "[1] Amplitude Min/Max"=<0 1>
      )
     )
     B=25
     Alpha=$System.Dive
    )
   ) {module=UpdateMeshOrientation001}
   &InterpolateOverTime(Rate Of Change=5, Target Quat Value:Quat4f=$Particles.MeshOrientation, Variable Type=::NewEnumerator5) {module=InterpolateOverTime002}
   /SetVariables(Particles.New Orient:Quat4f=$Particles.InterpolateOverTime002.InterpolateOverTime_Quaternion.Moving Average) {isSetParameters=true, module=SetVariables_#13}
   /SetVariables(Particles.MeshOrientation=$Particles.New Orient) {isSetParameters=true, module=SetVariables_#14}
  ]
  rendererClasses [&NiagaraMeshRendererProperties]
  renderers [{rendererClass=&NiagaraMeshRendererProperties, rendererIndex=0}]
 }
 Ship_FE {
  emitterUpdateScript [
   &EmitterState()
   &SpawnBurst_Instantaneous(
    Age:NiagaraFloat=$Emitter.Age
    Spawn Count=&MakeIntFromFloat(Float=&FN_FrontEndFloatModifier(Gameplay=0, Front End=3))
    Use Spawn Probability=false
    UseLoopCountLimit=false
   )
   &InterpolateOverTime(Rate Of Change=50, Target Vector Value:Vector3f=$Engine.Owner.Velocity, Variable Type=::NewEnumerator2)
  ]
  particleSpawnScript [
   &InitializeParticle(
    Color Mode=::NewEnumerator0
    Lifetime=5
    Mesh Scale Mode=::NewEnumerator3
    Mesh Uniform Scale=$System.Ship Scale
    Position Offset=<0 0 1000>
    UsePositionOffset=true
    Write Material Random=true
   )
   &ShapeLocation(MaskSpawnGroup=false, Override Seed=false, UseOwnerScale=true)
   &InitialMeshOrientation(Mesh Orientation Mode=::NewEnumerator7, Use Rotation Vector=true)
   /SetVariables(Particles.Debug sprite offset=50) {isSetParameters=true, module=SetVariables_#11}
  ]
  particleUpdateScript [
   &ParticleState(Kill Particles When Lifetime Has Elapsed=false, Let Infinitely Lived Particles Die When Emitter Deactivates=true)
   /SetVariables(Particles.Ambient:Vector3f=&Multiply_VectorByFloat(
     Vector=&MakeVector(
      X=&Waveform(
       "Waveform [1]"=::NewEnumerator12
       "[1] Input Position"=&RandomRangeFloat(Minimum=0, Maximum=1000)
       "[1] Frequency"=1
       "[1] Amplitude Min/Max"=<-50 50>
      )
      Y=&Waveform(
       "Waveform [1]"=::NewEnumerator12
       "[1] Input Position"=&RandomRangeFloat(Minimum=1200, Maximum=1800)
       "[1] Frequency"=1
       "[1] Amplitude Min/Max"=<-50 50>
      )
      Z=&Waveform(
       "Waveform [1]"=::NewEnumerator12
       "[1] Input Position"=&RandomRangeFloat(Minimum=2200, Maximum=2800)
       "[1] Frequency"=1
       "[1] Amplitude Min/Max"=<-50 50>
      )
     )
     Float=0.2
    )) {isSetParameters=true, module=SetVariables_#12}
   &SkeletalMeshLocation(
    Bone Sampling Mode=::NewEnumerator4
    MaskSpawnGroup=false
    Orientation Sampling=::NewEnumerator1
    Position Sampling=::NewEnumerator2
    Sampled Position Offset=&RotateVector(
     Vector To Rotate:Vector3f=&Add_Vector(
      A=$Particles.Ambient
      B=&Add_Vector(
       A=&SelectVectorFromArray(
        Direct Array Index=&ReturnParticleIDAttributes()
        Direct Array Mode=::NewEnumerator1
        Vector Selection Array=DataInterface() {propertyValues {FloatData [<0 75 165>, <-65 0 175>, <65 0 175>]}}
       )
       B=$System.OverallTargetOffsetFE
      )
     )
     Yaw:NiagaraFloat=&Waveform(
      "Waveform [1]"=::NewEnumerator12
      "[1] Input Position"=&RandomRangeFloat(Minimum=200000, Maximum=100000)
      "[1] Frequency"=1.25
      "[1] Amplitude Min/Max"=<5 -5>
     )
     Pitch:NiagaraFloat=0
     Roll:NiagaraFloat=0
     Rotation Rate:NiagaraFloat=1
    )
    Skeletal Mesh=$System.SkeletalMesh
    Use Sampled Mesh Attribute as Ribbon ID=false
   )
   &ScaleMeshSize(Scale Factor=&VectorFromFloat(Value=&FloatFromCurve(
      FloatCurve=DataInterface() {propertyValues {Curve {keys [{arriveTangent=1.0186, interpMode=RCIM_Cubic, leaveTangent=1.0186, tangentMode=RCTM_User, time=0, value=0}, <at 1 1 RCIM_Cubic>]}}}
      CurveIndex=$Particles.Age
     )))
   &DynamicMaterialParameters(
    Index 0 Param 1=1
    Module.Param0WriteEnabled=true
    Module.Param1WriteEnabled=true
    Module.Param2WriteEnabled=true
    Module.Param3WriteEnabled=true
   )
   &Drag(Drag=2, Use Linear Drag=true, Use Rotational Drag=false)
   &SolveForcesAndVelocity(Clamp Velocity=false, Limit Acceleration=false)
   &UpdateMeshOrientation(
    Bank Rate=900000000
    Banking Angle Min / Max:Vector2f=<0.1 -0.1>
    Orientation Method=::NewEnumerator3
    Particle Velocity:Vector3f=&Multiply_VectorByFloat(
     Vector=&Add_Vector(
      A=&ReturnMeshOrientationAxis(Mesh Orientation Axis=::YAxis, Mesh Orientation Quat=$Particles.SkeletalMeshLocation.SampledRotation)
      B=&Lerp_Vector(A=<0 0 -2>, B=<0 0 -0.5>, Alpha=$System.Dive)
     )
     Float=500000
    )
    Rotation Decay Rate In Degrees=900000000
    Up Vector:Vector3f=&Lerp_Vector(
     A=&ReturnMeshOrientationAxis(Mesh Orientation Axis=::YAxis, Mesh Orientation Quat=$Particles.SkeletalMeshLocation.SampledRotation)
     B=&Multiply_VectorByFloat(
      Vector=&ReturnMeshOrientationAxis(Mesh Orientation Axis=::ZAxis, Mesh Orientation Quat=$Particles.SkeletalMeshLocation.SampledRotation)
      Float=1
     )
     Alpha=&FloatFromCurve(FloatCurve=DataInterface() {propertyValues {Curve {keys [<at 0 0.5>,<at 0.25 0.75>]}}}, CurveIndex=$System.Dive)
    )
    Upright Roll Correction Speed=900000000
    Velocity Orientation Convergence Rate=900000000
   ) {module=UpdateMeshOrientation001}
   &InterpolateOverTime(
    Initial Vector Value:NiagaraPosition=$Engine.Owner.Position
    Rate Of Change=500
    Target Vector Value:NiagaraPosition=$Particles.SkeletalMeshLocation.SampledPosition
    Variable Type=::NewEnumerator2
   ) {module=InterpolateOverTime001}
   &InterpolateOverTime(Rate Of Change=5, Target Quat Value:Quat4f=$Particles.MeshOrientation, Variable Type=::NewEnumerator5) {module=InterpolateOverTime002}
   /SetVariables(Particles.Position=$Particles.InterpolateOverTime001.InterpolateOverTime_V3_.Moving Average) {isSetParameters=true}
   /SetVariables(Particles.New Orient:Quat4f=$Particles.InterpolateOverTime002.InterpolateOverTime_Quaternion.Moving Average) {isSetParameters=true, module=SetVariables_#13}
   /SetVariables(Particles.MeshOrientation=$Particles.New Orient) {isSetParameters=true, module=SetVariables_#14}
  ]
  rendererClasses [&NiagaraMeshRendererProperties]
  renderers [{rendererClass=&NiagaraMeshRendererProperties, rendererIndex=0}]
 }
 Ship_Trails001 {
  emitterUpdateScript [
   &EmitterState(Delay First Loop Only=true, Life Cycle Mode=::NewEnumerator1, Loop Delay=0.01, Loop Duration=1, UseLoopDelay=true)
   &SpawnParticlesFromOtherEmitter(
    Attribute Reader=DataInterface() {propertyValues {EmitterBinding {bindingMode=Other, emitterName=ship}}}
    Spawn Rate=&FN_FrontEndFloatModifier(Gameplay=&Lerp_Float(A=100, B=300, Alpha=$System.Dive), Front End=0)
    Spawn Rate Per Particle Cap=10000000
    Use Maximum Age=false
    Use Minimum Age=false
   )
  ]
  particleSpawnScript [
   &InitializeParticle(
    Color Mode=::NewEnumerator0
    Lifetime=&Lerp_Float(A=0.1, B=0.025, Alpha=$System.Dive)
    UsePositionOffset=false
    Write Material Random=true
   )
   &InitialMeshOrientation(Use Rotation Vector=false)
   &SampleParticlesFromOtherEmitter(
    Applied Velocity Scale=0
    Mesh Orientation Sampling:ENiagara_DirectReadSamplingMode=::NewEnumerator0
    Particle ID Sampling=::NewEnumerator1
    Velocity Application:ENiagara_DirectReadApplicationMode=::NewEnumerator0
    Velocity Sampling=::NewEnumerator0
   )
   /SetVariables(Particles.RibbonWidth=1.75) {isSetParameters=true, module=SetVariables_#15}
   /SetVariables(Particles.RibbonID=&self:RealmTrailID()) {isSetParameters=true, module=SetVariables_#16}
   &ShapeLocation(
    MaskSpawnGroup=false
    Offset=&Multiply_VectorByFloat(
     Vector=&ReturnMeshOrientationAxis(Mesh Orientation Axis=::XAxis, Mesh Orientation Quat=$Output.SampleParticlesFromOtherEmitter.SampledMeshOrientation)
     Float=6.5
    )
    Offset Mode=::NewEnumerator0
    Override Seed=false
    Sphere Radius=0
    UseOwnerScale=true
   )
  ]
  particleUpdateScript [
   &ParticleState()
   &ScaleColor(
    Scale Alpha=&FloatFromCurve(FloatCurve=DataInterface() {propertyValues {Curve {keys [
         {arriveTangent=-2.8469, interpMode=RCIM_Cubic, leaveTangent=-2.8469, tangentMode=RCTM_User, time=0, value=1}
         {arriveTangentWeight=0.2077, interpMode=RCIM_Cubic, tangentMode=RCTM_User, time=1, value=0}
        ]}}})
    Scale RGB=<2 2 2>
    ScaleA=true
    ScaleRGB=true
   )
   &SolveForcesAndVelocity(Clamp Velocity=false, Limit Acceleration=false)
   /SetVariables(Particles.LerpPos:NiagaraPosition=&Lerp_Position(
     A:NiagaraPosition=&self:ScratchDynamicInput_01(
      index:NiagaraInt32=&MakeIntFromFloat(Float=$Particles.SampleParticlesFromOtherEmitter.SampledIndex)
      New Particle Attribute Reader:NiagaraDataInterfaceParticleRead=DataInterface() {propertyValues {EmitterBinding {bindingMode=Other, emitterName=Ship}}}
     )
     B:NiagaraPosition=$Particles.Position
     Alpha:NiagaraFloat=&FloatFromCurve(FloatCurve=DataInterface() {propertyValues {Curve {keys [{tangentMode=RCTM_User, time=0.1, value=0}, {tangentMode=RCTM_User, time=0.25, value=1}]}}})
    )) {isSetParameters=true, module=SetVariables_#17}
   &self:SampleShip(New Particle Attribute Reader=DataInterface() {propertyValues {EmitterBinding {bindingMode=Other, emitterName=ship}}}) {module=SampleShip}
   &FindClosestPointOnLineSegment(
    Line Segment End:Vector3f=&Add_Vector(
     A=&self:ScratchDynamicInput_01(
      index:NiagaraInt32=&MakeIntFromFloat(Float=$Particles.SampleParticlesFromOtherEmitter.SampledIndex)
      New Particle Attribute Reader:NiagaraDataInterfaceParticleRead=DataInterface() {propertyValues {EmitterBinding {bindingMode=Other, emitterName=Ship}}}
     )
     B=&Multiply_VectorByFloat(
      Vector=&ReturnMeshOrientationAxis(Mesh Orientation Axis=::XAxis, Mesh Orientation Quat=$Output.SampleShip.Ship Rot)
      Float=-500
     )
    )
    Line Segment Start:Vector3f=&self:ScratchDynamicInput_01(
     index:NiagaraInt32=&MakeIntFromFloat(Float=$Particles.SampleParticlesFromOtherEmitter.SampledIndex)
     New Particle Attribute Reader:NiagaraDataInterfaceParticleRead=DataInterface() {propertyValues {EmitterBinding {bindingMode=Other, emitterName=Ship}}}
    )
    Sample Position:NiagaraPosition=$Particles.LerpPos
   )
   /SetVariables(Particles.Line Position:NiagaraPosition=&Lerp_Position(
     A:NiagaraPosition=Linked() {linkedVariable {name=Output.FindClosestPointOnLineSegment.Closest Point on Line Segment, type=Vector3f}}
     B:NiagaraPosition=$Particles.LerpPos
     Alpha:NiagaraFloat=&FloatFromCurve(FloatCurve=DataInterface() {propertyValues {Curve {keys [<at 0.1 0>,<at 1 1>]}}})
    )) {isSetParameters=true, module=SetVariables_#18}
   /SetVariables(Particles.Jitter line Positiuon:NiagaraPosition=&Lerp_Position(
     A:NiagaraPosition=$Particles.Line Position
     B:NiagaraPosition=&Add_Vector(
      A=Linked() {linkedVariable {name=Particles.Line Position, type=NiagaraPosition}}
      B=&RandomVector(Vector Scale=&Lerp_Float(A=0, B=0, Alpha=$System.Dive), Evaluation Type=::NewEnumerator1)
     )
     Alpha:NiagaraFloat=&FloatFromCurve(FloatCurve=DataInterface() {propertyValues {Curve {keys [
          {arriveTangent=4.3486, interpMode=RCIM_Cubic, leaveTangent=4.3486, tangentMode=RCTM_User, time=0, value=0}
          {tangentMode=RCTM_User, time=1, value=1}
         ]}}})
    )) {isSetParameters=true, module=SetVariables_#19}
   /SetVariables(Particles.Overall Offset Pos:NiagaraPosition=&Add_Vector(
     A=Linked() {linkedVariable {name=Particles.Jitter line Positiuon, type=NiagaraPosition}}
     B=&Multiply_VectorByFloat(
      Vector=&ReturnMeshOrientationAxis(Mesh Orientation Axis=::YAxis, Mesh Orientation Quat=$Output.SampleShip.Ship Rot)
      Float=&Multiply_Float(A=3.5, B=$System.Ship Scale)
     )
    )) {isSetParameters=true, module=SetVariables_#20}
  ]
  rendererClasses [&NiagaraRibbonRendererProperties]
  renderers [{rendererClass=&NiagaraRibbonRendererProperties}]
 }
 Ship_Trails002 {
  emitterUpdateScript [
   &EmitterState(Delay First Loop Only=true, Life Cycle Mode=::NewEnumerator1, Loop Delay=0.01, Loop Duration=1, UseLoopDelay=true)
   &SpawnParticlesFromOtherEmitter(
    Attribute Reader=DataInterface() {propertyValues {EmitterBinding {bindingMode=Other, emitterName=ship}}}
    Spawn Rate=&FN_FrontEndFloatModifier(Gameplay=&Lerp_Float(A=100, B=300, Alpha=$System.Dive), Front End=0)
    Spawn Rate Per Particle Cap=10000000
    Use Maximum Age=false
    Use Minimum Age=false
   )
  ]
  particleSpawnScript [
   &InitializeParticle(
    Color Mode=::NewEnumerator0
    Lifetime=&Lerp_Float(A=0.1, B=0.025, Alpha=$System.Dive)
    UsePositionOffset=false
    Write Material Random=true
   )
   &InitialMeshOrientation(Use Rotation Vector=false)
   &SampleParticlesFromOtherEmitter(
    Applied Velocity Scale=0
    Mesh Orientation Sampling:ENiagara_DirectReadSamplingMode=::NewEnumerator0
    Particle ID Sampling=::NewEnumerator1
    Velocity Application:ENiagara_DirectReadApplicationMode=::NewEnumerator0
    Velocity Sampling=::NewEnumerator0
   )
   /SetVariables(Particles.RibbonWidth=1.75) {isSetParameters=true, module=SetVariables_#15}
   /SetVariables(Particles.RibbonID=&self:RealmTrailID()) {isSetParameters=true, module=SetVariables_#16}
   &ShapeLocation(
    MaskSpawnGroup=false
    Offset=&Multiply_VectorByFloat(
     Vector=&ReturnMeshOrientationAxis(Mesh Orientation Axis=::XAxis, Mesh Orientation Quat=$Output.SampleParticlesFromOtherEmitter.SampledMeshOrientation)
     Float=6.5
    )
    Offset Mode=::NewEnumerator0
    Override Seed=false
    Sphere Radius=0
    UseOwnerScale=true
   )
  ]
  particleUpdateScript [
   &ParticleState()
   &ScaleColor(
    Scale Alpha=&FloatFromCurve(FloatCurve=DataInterface() {propertyValues {Curve {keys [
         {arriveTangent=-2.8469, interpMode=RCIM_Cubic, leaveTangent=-2.8469, tangentMode=RCTM_User, time=0, value=1}
         {arriveTangentWeight=0.2077, interpMode=RCIM_Cubic, tangentMode=RCTM_User, time=1, value=0}
        ]}}})
    Scale RGB=<2 2 2>
    ScaleA=true
    ScaleRGB=true
   )
   &SolveForcesAndVelocity(Clamp Velocity=false, Limit Acceleration=false)
   /SetVariables(Particles.LerpPos:NiagaraPosition=&Lerp_Position(
     A:NiagaraPosition=&self:ScratchDynamicInput_01(
      index:NiagaraInt32=&MakeIntFromFloat(Float=$Particles.SampleParticlesFromOtherEmitter.SampledIndex)
      New Particle Attribute Reader:NiagaraDataInterfaceParticleRead=DataInterface() {propertyValues {EmitterBinding {bindingMode=Other, emitterName=Ship}}}
     )
     B:NiagaraPosition=$Particles.Position
     Alpha:NiagaraFloat=&FloatFromCurve(FloatCurve=DataInterface() {propertyValues {Curve {keys [{tangentMode=RCTM_User, time=0.1, value=0}, {tangentMode=RCTM_User, time=0.25, value=1}]}}})
    )) {isSetParameters=true, module=SetVariables_#17}
   &self:SampleShip(New Particle Attribute Reader=DataInterface() {propertyValues {EmitterBinding {bindingMode=Other, emitterName=ship}}}) {module=SampleShip}
   &FindClosestPointOnLineSegment(
    Line Segment End:Vector3f=&Add_Vector(
     A=&self:ScratchDynamicInput_01(
      index:NiagaraInt32=&MakeIntFromFloat(Float=$Particles.SampleParticlesFromOtherEmitter.SampledIndex)
      New Particle Attribute Reader:NiagaraDataInterfaceParticleRead=DataInterface() {propertyValues {EmitterBinding {bindingMode=Other, emitterName=Ship}}}
     )
     B=&Multiply_VectorByFloat(
      Vector=&ReturnMeshOrientationAxis(Mesh Orientation Axis=::XAxis, Mesh Orientation Quat=$Output.SampleShip.Ship Rot)
      Float=-500
     )
    )
    Line Segment Start:Vector3f=&self:ScratchDynamicInput_01(
     index:NiagaraInt32=&MakeIntFromFloat(Float=$Particles.SampleParticlesFromOtherEmitter.SampledIndex)
     New Particle Attribute Reader:NiagaraDataInterfaceParticleRead=DataInterface() {propertyValues {EmitterBinding {bindingMode=Other, emitterName=Ship}}}
    )
    Sample Position:NiagaraPosition=$Particles.LerpPos
   )
   /SetVariables(Particles.Line Position:NiagaraPosition=&Lerp_Position(
     A:NiagaraPosition=Linked() {linkedVariable {name=Output.FindClosestPointOnLineSegment.Closest Point on Line Segment, type=Vector3f}}
     B:NiagaraPosition=$Particles.LerpPos
     Alpha:NiagaraFloat=&FloatFromCurve(FloatCurve=DataInterface() {propertyValues {Curve {keys [<at 0.1 0>,<at 1 1>]}}})
    )) {isSetParameters=true, module=SetVariables_#18}
   /SetVariables(Particles.Jitter line Positiuon:NiagaraPosition=&Lerp_Position(
     A:NiagaraPosition=$Particles.Line Position
     B:NiagaraPosition=&Add_Vector(
      A=Linked() {linkedVariable {name=Particles.Line Position, type=NiagaraPosition}}
      B=&RandomVector(Vector Scale=&Lerp_Float(A=0, B=0, Alpha=$System.Dive), Evaluation Type=::NewEnumerator1)
     )
     Alpha:NiagaraFloat=&FloatFromCurve(FloatCurve=DataInterface() {propertyValues {Curve {keys [
          {arriveTangent=4.3486, interpMode=RCIM_Cubic, leaveTangent=4.3486, tangentMode=RCTM_User, time=0, value=0}
          {tangentMode=RCTM_User, time=1, value=1}
         ]}}})
    )) {isSetParameters=true, module=SetVariables_#19}
   /SetVariables(Particles.Overall Offset Pos:NiagaraPosition=&Add_Vector(
     A=Linked() {linkedVariable {name=Particles.Jitter line Positiuon, type=NiagaraPosition}}
     B=&Multiply_VectorByFloat(
      Vector=&ReturnMeshOrientationAxis(Mesh Orientation Axis=::YAxis, Mesh Orientation Quat=$Output.SampleShip.Ship Rot)
      Float=&Multiply_Float(A=-3.5, B=$System.Ship Scale)
     )
    )) {isSetParameters=true, module=SetVariables_#20}
  ]
  rendererClasses [&NiagaraRibbonRendererProperties]
  renderers [{rendererClass=&NiagaraRibbonRendererProperties}]
 }
 Ship_Trails_2_FE {
  emitterUpdateScript [
   &EmitterState()
   &SpawnParticlesFromOtherEmitter(
    Attribute Reader=DataInterface() {propertyValues {EmitterBinding {bindingMode=Other, emitterName=ship_FE}}}
    Spawn Rate=&FN_FrontEndFloatModifier(Gameplay=0, Front End=&Lerp_Float(A=75, B=450, Alpha=$System.Dive))
    Spawn Rate Per Particle Cap=1000000
    Use Maximum Age=false
    Use Minimum Age=false
   )
  ]
  particleSpawnScript [
   &InitializeParticle(
    Color Mode=::NewEnumerator0
    Lifetime=&Lerp_Float(A=0.1, B=0.025, Alpha=$System.Dive)
    UsePositionOffset=false
    Write Material Random=true
   )
   &InitialMeshOrientation(Use Rotation Vector=false)
   &SampleParticlesFromOtherEmitter(
    Applied Velocity Scale=0
    Mesh Orientation Sampling:ENiagara_DirectReadSamplingMode=::NewEnumerator0
    Particle ID Sampling=::NewEnumerator1
    Velocity Application:ENiagara_DirectReadApplicationMode=::NewEnumerator0
    Velocity Sampling=::NewEnumerator0
   )
   /SetVariables(Particles.RibbonWidth=1.75) {isSetParameters=true, module=SetVariables_#15}
   /SetVariables(Particles.RibbonID=&self:RealmTrailID()) {isSetParameters=true, module=SetVariables_#16}
   &ShapeLocation(
    MaskSpawnGroup=false
    Offset=&Multiply_VectorByFloat(
     Vector=&ReturnMeshOrientationAxis(Mesh Orientation Axis=::XAxis, Mesh Orientation Quat=$Output.SampleParticlesFromOtherEmitter.SampledMeshOrientation)
     Float=6.5
    )
    Offset Mode=::NewEnumerator0
    Override Seed=false
    Sphere Radius=0
    UseOwnerScale=true
   )
   &AddVelocity(
    Rotation Coordinate Space=::World
    Velocity=&ReturnMeshOrientationAxis(Mesh Orientation Axis=::XAxis, Mesh Orientation Quat=$Output.SampleParticlesFromOtherEmitter.SampledMeshOrientation)
    Velocity Speed Scale=&Lerp_Float(A=-750, B=-9000, Alpha=$System.Dive)
   )
  ]
  particleUpdateScript [
   &ParticleState()
   &ScaleColor(
    Scale Alpha=&FloatFromCurve(FloatCurve=DataInterface() {propertyValues {Curve {keys [
         {arriveTangent=-2.8469, interpMode=RCIM_Cubic, leaveTangent=-2.8469, tangentMode=RCTM_User, time=0, value=1}
         {arriveTangentWeight=0.2077, interpMode=RCIM_Cubic, tangentMode=RCTM_User, time=1, value=0}
        ]}}})
    Scale RGB=<2 2 2>
    ScaleA=true
    ScaleRGB=true
   )
   &SolveForcesAndVelocity(Clamp Velocity=false, Limit Acceleration=false)
   /SetVariables(Particles.LerpPos:NiagaraPosition=&Lerp_Position(
     A:NiagaraPosition=&self:ScratchDynamicInput_01(
      index:NiagaraInt32=&MakeIntFromFloat(Float=$Particles.SampleParticlesFromOtherEmitter.SampledIndex)
      New Particle Attribute Reader:NiagaraDataInterfaceParticleRead=DataInterface() {propertyValues {EmitterBinding {bindingMode=Other, emitterName=ship_FE}}}
     )
     B:NiagaraPosition=$Particles.Position
     Alpha:NiagaraFloat=&FloatFromCurve(FloatCurve=DataInterface() {propertyValues {Curve {keys [{tangentMode=RCTM_User, time=0.1, value=0}, {tangentMode=RCTM_User, time=0.25, value=1}]}}})
    )) {isSetParameters=true, module=SetVariables_#17}
   &self:SampleShip(New Particle Attribute Reader=DataInterface() {propertyValues {EmitterBinding {bindingMode=Other, emitterName=ship_FE}}}) {module=SampleShip}
   /SetVariables(Particles.Overall Offset Pos:NiagaraPosition=&Add_Vector(
     A=Linked() {linkedVariable {name=Particles.LerpPos, type=NiagaraPosition}}
     B=&Multiply_VectorByFloat(
      Vector=&ReturnMeshOrientationAxis(Mesh Orientation Axis=::YAxis, Mesh Orientation Quat=$Output.SampleShip.Ship Rot)
      Float=&Multiply_Float(A=-3.5, B=$System.Ship Scale)
     )
    )) {isSetParameters=true, module=SetVariables_#20}
  ]
  rendererClasses [&NiagaraRibbonRendererProperties]
  renderers [{rendererClass=&NiagaraRibbonRendererProperties}]
 }
 Ship_Trails_2_FE001 {
  emitterUpdateScript [
   &EmitterState()
   &SpawnParticlesFromOtherEmitter(
    Attribute Reader=DataInterface() {propertyValues {EmitterBinding {bindingMode=Other, emitterName=ship_FE}}}
    Spawn Rate=&FN_FrontEndFloatModifier(Gameplay=0, Front End=&Lerp_Float(A=75, B=450, Alpha=$System.Dive))
    Spawn Rate Per Particle Cap=1000000
    Use Maximum Age=false
    Use Minimum Age=false
   )
  ]
  particleSpawnScript [
   &InitializeParticle(
    Color Mode=::NewEnumerator0
    Lifetime=&Lerp_Float(A=0.1, B=0.025, Alpha=$System.Dive)
    UsePositionOffset=false
    Write Material Random=true
   )
   &InitialMeshOrientation(Use Rotation Vector=false)
   &SampleParticlesFromOtherEmitter(
    Applied Velocity Scale=0
    Mesh Orientation Sampling:ENiagara_DirectReadSamplingMode=::NewEnumerator0
    Particle ID Sampling=::NewEnumerator1
    Velocity Application:ENiagara_DirectReadApplicationMode=::NewEnumerator0
    Velocity Sampling=::NewEnumerator0
   )
   /SetVariables(Particles.RibbonWidth=1.75) {isSetParameters=true, module=SetVariables_#15}
   /SetVariables(Particles.RibbonID=&self:RealmTrailID()) {isSetParameters=true, module=SetVariables_#16}
   &ShapeLocation(
    MaskSpawnGroup=false
    Offset=&Multiply_VectorByFloat(
     Vector=&ReturnMeshOrientationAxis(Mesh Orientation Axis=::XAxis, Mesh Orientation Quat=$Output.SampleParticlesFromOtherEmitter.SampledMeshOrientation)
     Float=6.5
    )
    Offset Mode=::NewEnumerator0
    Override Seed=false
    Sphere Radius=0
    UseOwnerScale=true
   )
   &AddVelocity(
    Rotation Coordinate Space=::World
    Velocity=&ReturnMeshOrientationAxis(Mesh Orientation Axis=::XAxis, Mesh Orientation Quat=$Output.SampleParticlesFromOtherEmitter.SampledMeshOrientation)
    Velocity Speed Scale=&Lerp_Float(A=-750, B=-9000, Alpha=$System.Dive)
   )
  ]
  particleUpdateScript [
   &ParticleState()
   &ScaleColor(
    Scale Alpha=&FloatFromCurve(FloatCurve=DataInterface() {propertyValues {Curve {keys [
         {arriveTangent=-2.8469, interpMode=RCIM_Cubic, leaveTangent=-2.8469, tangentMode=RCTM_User, time=0, value=1}
         {arriveTangentWeight=0.2077, interpMode=RCIM_Cubic, tangentMode=RCTM_User, time=1, value=0}
        ]}}})
    Scale RGB=<2 2 2>
    ScaleA=true
    ScaleRGB=true
   )
   &SolveForcesAndVelocity(Clamp Velocity=false, Limit Acceleration=false)
   /SetVariables(Particles.LerpPos:NiagaraPosition=&Lerp_Position(
     A:NiagaraPosition=&self:ScratchDynamicInput_01(
      index:NiagaraInt32=&MakeIntFromFloat(Float=$Particles.SampleParticlesFromOtherEmitter.SampledIndex)
      New Particle Attribute Reader:NiagaraDataInterfaceParticleRead=DataInterface() {propertyValues {EmitterBinding {bindingMode=Other, emitterName=ship_FE}}}
     )
     B:NiagaraPosition=$Particles.Position
     Alpha:NiagaraFloat=&FloatFromCurve(FloatCurve=DataInterface() {propertyValues {Curve {keys [{tangentMode=RCTM_User, time=0.1, value=0}, {tangentMode=RCTM_User, time=0.25, value=1}]}}})
    )) {isSetParameters=true, module=SetVariables_#17}
   &self:SampleShip(New Particle Attribute Reader=DataInterface() {propertyValues {EmitterBinding {bindingMode=Other, emitterName=ship_FE}}}) {module=SampleShip}
   /SetVariables(Particles.Overall Offset Pos:NiagaraPosition=&Add_Vector(
     A=Linked() {linkedVariable {name=Particles.LerpPos, type=NiagaraPosition}}
     B=&Multiply_VectorByFloat(
      Vector=&ReturnMeshOrientationAxis(Mesh Orientation Axis=::YAxis, Mesh Orientation Quat=$Output.SampleShip.Ship Rot)
      Float=&Multiply_Float(A=3.5, B=$System.Ship Scale)
     )
    )) {isSetParameters=true, module=SetVariables_#20}
  ]
  rendererClasses [&NiagaraRibbonRendererProperties]
  renderers [{rendererClass=&NiagaraRibbonRendererProperties}]
 }
}
summary {
 systemName=NS_Realm_Contrail
 emitters [
  {emitterName=Ship, rendererClasses [&NiagaraMeshRendererProperties]}
  {emitterName=FN_SpeedLines, simTarget=GPUComputeSim, rendererClasses [&NiagaraSpriteRendererProperties]}
  {emitterName=Ship_Trails002, rendererClasses [&NiagaraRibbonRendererProperties]}
  {emitterName=Ship_Trails001, rendererClasses [&NiagaraRibbonRendererProperties]}
  {emitterName=Air_Edges, rendererClasses [&NiagaraRibbonRendererProperties]}
  {emitterName=Ship_FE, rendererClasses [&NiagaraMeshRendererProperties]}
  {emitterName=Ship_Trails_2_FE, rendererClasses [&NiagaraRibbonRendererProperties]}
  {emitterName=Air_Edges_FE, rendererClasses [&NiagaraRibbonRendererProperties]}
  {emitterName=Ship_Trails_2_FE001, rendererClasses [&NiagaraRibbonRendererProperties]}
 ]
 userVariables [
  User.bIsFrontEnd {type=NiagaraBool}
  User.bIsFrontEndPreview {type=NiagaraBool}
  User.ParaGlideLeanAlpha {type=NiagaraFloat, defaultValue=0, description=""}
  User.PlayerSpeed {type=NiagaraFloat, defaultValue=0, description=""}
  User.RibbonIDOffset {type=NiagaraInt32}
 ]
}
